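# Physics asset for the back mesh of the medium female IonVial player body, covering a bottom armor piece and the upper spine it hangs from.
# A single box body on dyn_hackBottomArmor is constrained to a spine_03 capsule, which is constrained in turn to a spine_02 capsule.
# The armor joint has narrower swing limits than the spine joint and a zero twist limit.
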PhysicsAsset /Game/Characters/Player/Female/Medium/Bodies/F_MED_IonVial/Meshes/F_MED_IonVial_back_Physics
bodies {
 dyn_hackBottomArmor [{
   shapeType=Box
   extentX=9.5629
   extentY=2.174
   extentZ=12.42
   shapeName=dyn_hackBottomArmor_box
   center=<7.4121 0.6459 0>
   rotation=<rot 0 5.0001 0>
  }]
 spine_02 [{
   shapeType=Capsule
   length=12.1881
   radius=5.1394
   shapeName=spine_02_capsule
   center=<0.0174 -0.6602 -0.0956>
   rotation=<rot -0.6774 -90.0123 -0.115>
  }]
 spine_03 [{
   shapeType=Capsule
   length=8.4865
   radius=6.0821
   shapeName=spine_03_capsule
   center=<-0.2298 0.3289 0.0086>
   rotation=<rot 5.9394 -92.1957 -1.972>
  }]
}
constraints [
 {bone1Name=dyn_hackBottomArmor, bone2Name=spine_03, swing1LimitDegrees=23, swing2LimitDegrees=25, twistLimitDegrees=0}
 {bone1Name=spine_03, bone2Name=spine_02, swing1LimitDegrees=45, swing2LimitDegrees=45, twistLimitDegrees=45}
]
