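# Physics asset for the left-leg mesh of the medium female IonVial player body, covering the leg and a lower-dress bone chain.
# Capsules cover the thigh and calf, plus three dyn_lowerDress bones, with the middle dress bone using three differently angled capsules.
# The dress bones chain from the thigh with tight swing limits that loosen slightly toward the tip, while the calf-to-thigh joint allows much wider swing.
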PhysicsAsset /Game/Characters/Player/Female/Medium/Bodies/F_MED_IonVial/Meshes/F_MED_IonVial_legL_Physics
bodies {
 calf_l [{
   shapeType=Capsule
   length=17.5919
   radius=6.4891
   shapeName=calf_l_capsule
   center=<-10.9525 -2.142 0.0775>
   rotation=<rot -1.2976 100.8563 -88.6044>
  }]
 dyn_lowerDress_01 [{
   shapeType=Capsule
   length=7.848
   radius=3.122
   shapeName=dyn_lowerDress_01_capsule
   center=<4.474 -0.1772 2.7787>
   rotation=<rot -81.3467 -102.6567 97.9308>
  }]
 dyn_lowerDress_02 [
  {
   shapeType=Capsule
   length=4.6123
   radius=2.4231
   shapeName=dyn_lowerDress_02_capsule
   center=<3.2237 -0.2509 1.5006>
   rotation=<rot -12.4565 89.0868 -90.3841>
  }
  {
   shapeType=Capsule
   length=7.883
   radius=2.5616
   shapeName=dyn_lowerDress_02_capsule
   center=<2.6136 7.4616 -5.2918>
   rotation=<rot -63.5223 -1.5903 -10.0897>
  }
  {
   shapeType=Capsule
   length=12.8799
   radius=2.5013
   shapeName=dyn_lowerDress_02_capsule
   center=<1.1148 -9.5328 -2.9951>
   rotation=<rot 69.0584 171.0197 151.7418>
  }
 ]
 dyn_lowerDress_03 [{
   shapeType=Capsule
   length=5.6109
   radius=2.3275
   shapeName=dyn_lowerDress_03_capsule
   center=<2.0924 -0.5885 0.3638>
   rotation=<rot -13.7908 86.4513 -94.9523>
  }]
 thigh_l [{
   shapeType=Capsule
   length=19.4909
   radius=7.9359
   shapeName=thigh_l_capsule
   center=<-20.0385 -0.1213 0.9454>
   rotation=<rot 0.6962 -89.7546 -89.5136>
  }]
}
constraints [
 {bone1Name=calf_l, bone2Name=thigh_l, swing1LimitDegrees=45, swing2LimitDegrees=45, twistLimitDegrees=0}
 {bone1Name=dyn_lowerDress_03, bone2Name=dyn_lowerDress_02, swing1LimitDegrees=7, swing2LimitDegrees=7, twistLimitDegrees=0}
 {bone1Name=dyn_lowerDress_02, bone2Name=dyn_lowerDress_01, swing1LimitDegrees=6, swing2LimitDegrees=6, twistLimitDegrees=0}
 {bone1Name=dyn_lowerDress_01, bone2Name=thigh_l, swing1LimitDegrees=5, swing2LimitDegrees=5, twistLimitDegrees=0}
]
