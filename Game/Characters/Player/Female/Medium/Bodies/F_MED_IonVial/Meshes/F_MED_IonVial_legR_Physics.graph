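# Physics asset for the right-leg mesh of the female medium IonVial player body, rigging a dynamic belt that hangs from the thigh.
# Front and back chains of capsule bodies start at base bodies angularly locked to thigh_r, with narrow swing and twist limits along each chain.
# A free-swing constraint joins the two chain ends, closing the belt into a loop.
# Pelvis and thigh capsules, angularly locked together, provide the body volumes around it.
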PhysicsAsset /Game/Characters/Player/Female/Medium/Bodies/F_MED_IonVial/Meshes/F_MED_IonVial_legR_Physics
bodies {
 dyn_belt [{
   shapeType=Capsule
   length=2.9822
   radius=1.6826
   shapeName=dyn_belt_capsule
   center=<3.1853 1.0384 -0.2959>
   rotation=<rot 89.1869 -136.5633 39.1421>
  }]
 dyn_belt_back_01 [{
   shapeType=Capsule
   length=4.7894
   radius=1.8484
   shapeName=dyn_belt_back_01_capsule
   center=<2.9626 0.6748 0.1161>
   rotation=<rot -1.1277 90.8975 -89.9767>
  }]
 dyn_belt_back_02 [{
   shapeType=Capsule
   length=2.8668
   radius=2.0347
   shapeName=dyn_belt_back_02_capsule
   center=<3.3885 0.685 0.354>
   rotation=<rot -11.9775 -89.4798 -89.3536>
  }]
 dyn_belt_back_base [{shapeType=Capsule, length=0.505, radius=0.505, shapeName=dyn_belt_back_base_capsule, rotation=<rot 0 -90 0>}]
 dyn_belt_back_end [{shapeType=Capsule, length=3.6243, radius=0.8901, center=<0.1722 0 -0.3268>, rotation=<rot -5.2268 0.0002 0.0003>}]
 dyn_belt_front_01 [{
   shapeType=Capsule
   length=2.0719
   radius=2.4518
   shapeName=dyn_belt_front_01_capsule
   center=<3.4141 -0.6498 -0.0134>
   rotation=<rot 0.3661 -96.6546 90.074>
  }]
 dyn_belt_front_02 [{
   shapeType=Capsule
   length=1.2375
   radius=2.319
   shapeName=dyn_belt_front_02_capsule
   center=<2.7325 -1.1501 0.2585>
   rotation=<rot -1.4645 80.04 -87.6318>
  }]
 dyn_belt_front_base [{shapeType=Capsule, length=0.505, radius=0.505, shapeName=dyn_belt_front_base_capsule, center=<0 -0.1072 0>, rotation=<rot -2.1781 -90 0>}]
 dyn_belt_front_end [{shapeType=Capsule, length=3.6243, radius=0.8901, center=<0.1722 0 -0.3268>, rotation=<rot -5.2268 0.0002 0.0003>}]
 pelvis [{
   shapeType=Capsule
   length=10.3177
   radius=9.494
   shapeName=pelvis_capsule
   center=<-2.2769 -1.499 -0.7807>
   rotation=<rot -5.3194 -88.3409 -0.6101>
  }]
 thigh_r [
  {
   shapeType=Capsule
   length=14.9858
   radius=5.8195
   shapeName=thigh_r_capsule
   center=<13.4327 2.9587 -0.3722>
   rotation=<rot -9.7807 78.2583 -89.7818>
  }
  {
   shapeType=Capsule
   length=8.7006
   radius=7.2601
   shapeName=thigh_r_capsule
   center=<14.9732 0.0359 -0.9831>
   rotation=<rot -80 -90.0003 86.2542>
  }
 ]
}
constraints [
 {bone1Name=dyn_belt_back_02, bone2Name=dyn_belt_back_01, swing1LimitDegrees=8, swing2LimitDegrees=12, twistLimitDegrees=3}
 {bone1Name=dyn_belt, bone2Name=dyn_belt_back_02, swing1LimitDegrees=5, swing2LimitDegrees=7, twistLimitDegrees=3}
 {bone1Name=dyn_belt_front_01, bone2Name=dyn_belt_front_base, swing1LimitDegrees=7, swing2LimitDegrees=9, twistLimitDegrees=3}
 {bone1Name=dyn_belt_front_02, bone2Name=dyn_belt_front_01, swing1LimitDegrees=6, swing2LimitDegrees=8, twistLimitDegrees=3}
 {bone1Name=dyn_belt_back_end, bone2Name=dyn_belt, swing1LimitDegrees=5, swing2LimitDegrees=4, twistLimitDegrees=5}
 {bone1Name=dyn_belt_front_end, bone2Name=dyn_belt_front_02, swing1LimitDegrees=8, swing2LimitDegrees=8, twistLimitDegrees=5}
 {
  bone1Name=dyn_belt_back_end
  bone2Name=dyn_belt_front_end
  swing1LimitDegrees=45
  swing1Motion=Free
  swing2LimitDegrees=45
  swing2Motion=Free
  twistLimitDegrees=5
 }
 {
  bone1Name=thigh_r
  bone2Name=pelvis
  swing1LimitDegrees=45
  swing1Motion=Locked
  swing2LimitDegrees=45
  swing2Motion=Locked
  twistLimitDegrees=45
  twistMotion=Locked
 }
 {bone1Name=dyn_belt_back_01, bone2Name=dyn_belt_back_base, swing1LimitDegrees=15, swing2LimitDegrees=20, twistLimitDegrees=0}
 {
  bone1Name=dyn_belt_front_base
  bone2Name=thigh_r
  swing1LimitDegrees=45
  swing1Motion=Locked
  swing2LimitDegrees=45
  swing2Motion=Locked
  twistLimitDegrees=45
  twistMotion=Locked
 }
 {
  bone1Name=dyn_belt_back_base
  bone2Name=thigh_r
  swing1LimitDegrees=45
  swing1Motion=Locked
  swing2LimitDegrees=45
  swing2Motion=Locked
  twistLimitDegrees=45
  twistMotion=Locked
 }
]
